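# Physics asset for the TruckScale female medium character's hair and accessory mesh part, giving two long left and right hair strands physical motion and collision.
# Capsules cover both strands plus head, clavicles, upper arms, pelvis and upper spine.
# Each side chains segments 2–6 from the head with tight limits, and 7–10 separately with looser limits.
# Small MSTR_Hair bodies attach to segment 6 with free rotation; no constraint links 6 to 7.
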PhysicsAsset /BRCosmetics/Characters/Player/Female/Medium/Bodies/F_MED_TruckScale_Hair/Meshes/Parts/F_MED_TruckScale_Hair_FaceAcc_Backpack_Physics
bodies {
 Hair_l_10 [{
   shapeType=Capsule
   length=8.1595
   radius=1.7438
   shapeName=Hair_l_10_capsule
   center=<5.1187 0.3448 -0.4002>
   rotation=<rot 0 -88.1094 -85.4421>
  }]
 Hair_l_2 [{
   shapeType=Capsule
   length=2.3275
   radius=3.2935
   shapeName=Hair_l_2_capsule
   center=<2.3306 0.0009 0.7429>
   rotation=<rot 8.7668 -89.9999 -83.8933>
  }]
 Hair_l_3 [{
   shapeType=Capsule
   length=3.2902
   radius=3.6583
   shapeName=Hair_l_3_capsule
   center=<2.6563 0.0603 0.3582>
   rotation=<rot 0.0001 -95.7536 -83.94>
  }]
 Hair_l_4 [{
   shapeType=Capsule
   length=2.1061
   radius=3.5027
   shapeName=Hair_l_4_capsule
   center=<3.5159 0.1475 0.1601>
   rotation=<rot 0.5082 -94.5985 -84.1813>
  }]
 Hair_l_5 [{shapeType=Capsule, length=4.7824, radius=2.0479, shapeName=Hair_l_5_capsule, center=<3.6617 0.0357 0.2256>, rotation=<rot 0 93.532 -89.9463>}]
 Hair_l_6 [{
   shapeType=Capsule
   length=7.2627
   radius=1.9577
   shapeName=Hair_l_6_capsule
   center=<4.6256 0.0805 0.0602>
   rotation=<rot 0 91.3184 -87.7702>
  }]
 Hair_l_7 [{
   shapeType=Capsule
   length=5.8174
   radius=1.8239
   shapeName=Hair_l_7_capsule
   center=<3.7782 0.0281 0.0691>
   rotation=<rot 0 91.6923 -89.8316>
  }]
 Hair_l_8 [{
   shapeType=Capsule
   length=5.955
   radius=1.6716
   shapeName=Hair_l_8_capsule
   center=<3.4798 0.0901 0.105>
   rotation=<rot 0.1223 89.8146 -88.7587>
  }]
 Hair_l_9 [{
   shapeType=Capsule
   length=3.8985
   radius=1.8483
   shapeName=Hair_l_9_capsule
   center=<4.1944 0.0452 -0.0858>
   rotation=<rot 0.0001 92.1507 -93.6036>
  }]
 Hair_r_10 [{
   shapeType=Capsule
   length=8.8634
   radius=1.6703
   shapeName=Hair_r_10_capsule
   center=<5.962 -0.2891 -0.7469>
   rotation=<rot 0 -90.1423 -82.6019>
  }]
 Hair_r_2 [{
   shapeType=Capsule
   length=2.2545
   radius=3.4198
   shapeName=Hair_r_2_capsule
   center=<2.5397 0.009 0.6406>
   rotation=<rot -5.5046 -90 -102.0299>
  }]
 Hair_r_3 [{shapeType=Capsule, length=3.3717, radius=3.7362, shapeName=Hair_r_3_capsule, center=<2.719 -0.0409 0.3839>, rotation=<rot 0 -85.8309 -84.13>}]
 Hair_r_4 [{
   shapeType=Capsule
   length=2.56
   radius=3.6277
   shapeName=Hair_r_4_capsule
   center=<3.4185 -0.1449 0.1134>
   rotation=<rot 0 -90.7075 -86.1601>
  }]
 Hair_r_5 [{shapeType=Capsule, length=5.2364, radius=2.0479, shapeName=Hair_r_5_capsule, center=<3.7644 -0.263 0.304>, rotation=<rot 0 -97.2147 -87.804>}]
 Hair_r_6 [{
   shapeType=Capsule
   length=6.1733
   radius=2.2703
   shapeName=Hair_r_6_capsule
   center=<4.889 -0.46 0.0042>
   rotation=<rot -0.179 93.226 -90.7664>
  }]
 Hair_r_7 [{
   shapeType=Capsule
   length=4.6499
   radius=1.8239
   shapeName=Hair_r_7_capsule
   center=<3.3685 -0.0772 0.0242>
   rotation=<rot 0 92.2056 -84.7212>
  }]
 Hair_r_8 [{
   shapeType=Capsule
   length=6.6544
   radius=1.5748
   shapeName=Hair_r_8_capsule
   center=<2.9058 -0.1573 0.2295>
   rotation=<rot 0.5674 90.7768 -91.4999>
  }]
 Hair_r_9 [{
   shapeType=Capsule
   length=3.81
   radius=1.6865
   shapeName=Hair_r_9_capsule
   center=<3.5491 -0.0122 0.0009>
   rotation=<rot -0.0048 88.7232 -89.904>
  }]
 MSTR_Hair_l [{shapeType=Capsule, length=0.505, radius=0.505, shapeName=MSTR_Hair_l_capsule, rotation=<rot 0 -90 0>}]
 MSTR_Hair_r [{shapeType=Capsule, length=0.505, radius=0.505, shapeName=MSTR_Hair_r_capsule, rotation=<rot 0 -90 0>}]
 clavicle_l [{
   shapeType=Capsule
   length=8.9697
   radius=9.066
   shapeName=clavicle_l_capsule
   center=<10.2115 2.4696 -1.1294>
   rotation=<rot -0.0004 -81.5625 90>
  }]
 clavicle_r [{shapeType=Capsule, length=10.4923, radius=7.041, shapeName=clavicle_r_capsule, center=<-4.8588 0.3079 0.8239>, rotation=<rot -0.0007 -90 90>}]
 head [{shapeType=Capsule, length=5, radius=9, center=<4.7224 2.1688 0>, rotation=<rot 90 -0.0001 0>}]
 pelvis [{shapeType=Capsule, length=30.005, radius=26.63, shapeName=pelvis_capsule, center=<-21.8827 7.1699 0>, rotation=<rot 0 -61.8758 90>}]
 spine_05 [
  {shapeType=Capsule, length=54.0595, radius=21.432, shapeName=spine_04_capsule, center=<-0.3119 15.1924 0>, rotation=<rot 0 -85.0009 0>}
  {shapeType=Capsule, length=29.5595, radius=36.1288, shapeName=spine_05_capsule, center=<-2.6852 -48.9373 0.9756>, rotation=<rot -90 0 -0.001>}
 ]
 upperarm_l [{
   shapeType=Capsule
   length=27.5109
   radius=7.9483
   shapeName=upperarm_l_capsule
   center=<19.7922 0.0001 0>
   rotation=<rot -0.0002 -89.9998 90>
  }]
 upperarm_r [{
   shapeType=Capsule
   length=23.0139
   radius=5.9345
   shapeName=upperarm_r_capsule
   center=<-14.0929 0 0.4319>
   rotation=<rot -0.0003 -89.9997 92.8125>
  }]
}
constraints [
 {bone1Name=Hair_l_9, bone2Name=Hair_l_8, swing1LimitDegrees=15, swing2LimitDegrees=15, twistLimitDegrees=4}
 {bone1Name=Hair_l_8, bone2Name=Hair_l_7, swing1LimitDegrees=15, swing2LimitDegrees=15, twistLimitDegrees=4}
 {bone1Name=Hair_l_6, bone2Name=Hair_l_5, swing1LimitDegrees=10, swing2LimitDegrees=10, twistLimitDegrees=1}
 {bone1Name=Hair_l_5, bone2Name=Hair_l_4, swing1LimitDegrees=10, swing2LimitDegrees=10, twistLimitDegrees=1}
 {bone1Name=Hair_l_4, bone2Name=Hair_l_3, swing1LimitDegrees=10, swing2LimitDegrees=10, twistLimitDegrees=1}
 {bone1Name=Hair_l_3, bone2Name=Hair_l_2, swing1LimitDegrees=10, swing2LimitDegrees=10, twistLimitDegrees=1}
 {bone1Name=Hair_r_2, bone2Name=Hair_r_3, swing1LimitDegrees=10, swing2LimitDegrees=10, twistLimitDegrees=1}
 {bone1Name=Hair_r_4, bone2Name=Hair_r_5, swing1LimitDegrees=10, swing2LimitDegrees=10, twistLimitDegrees=1}
 {bone1Name=Hair_r_5, bone2Name=Hair_r_6, swing1LimitDegrees=10, swing2LimitDegrees=10, twistLimitDegrees=1}
 {bone1Name=Hair_r_7, bone2Name=Hair_r_8, swing1LimitDegrees=15, swing2LimitDegrees=15, twistLimitDegrees=4}
 {bone1Name=Hair_r_8, bone2Name=Hair_r_9, swing1LimitDegrees=15, swing2LimitDegrees=15, twistLimitDegrees=4}
 {bone1Name=Hair_l_2, bone2Name=head, swing1LimitDegrees=10, swing2LimitDegrees=10, twistLimitDegrees=1}
 {bone1Name=Hair_r_2, bone2Name=head, swing1LimitDegrees=10, swing2LimitDegrees=10, twistLimitDegrees=1}
 {bone1Name=Hair_r_3, bone2Name=Hair_r_4, swing1LimitDegrees=10, swing2LimitDegrees=10, twistLimitDegrees=1}
 {bone1Name=Hair_l_10, bone2Name=Hair_l_9, swing1LimitDegrees=15, swing2LimitDegrees=15, twistLimitDegrees=4}
 {bone1Name=Hair_r_10, bone2Name=Hair_r_9, swing1LimitDegrees=15, swing2LimitDegrees=15, twistLimitDegrees=4}
 {
  bone1Name=MSTR_Hair_l
  bone2Name=Hair_l_6
  swing1LimitDegrees=45
  swing1Motion=Free
  swing2LimitDegrees=45
  swing2Motion=Free
  twistLimitDegrees=45
  twistMotion=Free
 }
 {
  bone1Name=MSTR_Hair_r
  bone2Name=Hair_r_6
  swing1LimitDegrees=45
  swing1Motion=Free
  swing2LimitDegrees=45
  swing2Motion=Free
  twistLimitDegrees=45
  twistMotion=Free
 }
]
