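# Control Rig for the alternate-skin TruckScale female hair that adjusts the hair_alt bones as the body moves.
# A four-way sequence applies fixed offsets and a head/spine_05 parent constraint, then corrective offsets for head, upper-spine, clavicle and pelvis poses.
# Spherical pose readers weight each corrective, with left and right clavicle readings merged by max.
# Control-to-bone transforms for head and spine_05 exist but are not wired into execution.
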ControlRigBlueprint /BRCosmetics/Characters/Player/Female/Medium/Bodies/F_MED_TruckScale_Hair/Meshes/Parts/F_MED_TruckScale_Hair_FaceAcc_CtrlRig_AltSkin
graphs [
 RigVMFunctionLibrary {links [], nodes []}
 RigVMFunction_Sequence_1_ContainedGraph {nodes [
   n1 = RigVMFunctionEntryNode()
   RigVMFunction_Sequence = &RigVMFunction_Sequence(ExecuteContext=@n1.ExecuteContext)
   n2 = &RigVMFunction_Sequence(ExecuteContext=@RigVMFunction_Sequence.B)
   n3 = &RigVMFunction_Sequence(ExecuteContext=@n2.B)
   RigVMFunctionReturnNode(A=@RigVMFunction_Sequence.A, B=@n2.A, C=@n3.A, D=@n3.B)
  ]}
 RigVMModel {nodes [
   RigUnit_BeginExecution = &RigUnit_BeginExecution()
   RigVMFunction_Sequence = RigVMAggregateNode(ExecuteContext=@RigUnit_BeginExecution.ExecutePin)
   RigUnit_OffsetTransformForItem = &RigUnit_OffsetTransformForItem(
    Item=(Type=Bone,Name="hair_alt_2")
    OffsetTransform=(Rotation=(X=0,Y=0,Z=-0.1305,W=0.9914),Translation=(X=0,Y=4,Z=0),Scale3D=(X=1,Y=1,Z=1))
    Weight=1
    bPropagateToChildren=True
    ExecutePin=@RigVMFunction_Sequence.A
   )
   > &RigUnit_OffsetTransformForItem(
    Item=(Type=Bone,Name="hair_alt_3")
    OffsetTransform=(Rotation=(X=0,Y=0,Z=-0.1305,W=0.9914),Translation=(X=0,Y=8,Z=0),Scale3D=(X=1,Y=1,Z=1))
    Weight=1
    bPropagateToChildren=True
   )
   > &RigUnit_ParentConstraint(
    AdvancedSettings=(InterpolationType=Average,RotationOrderForFilter=XZY)
    Child=(Type=Bone,Name="hair_alt_1")
    Filter=(TranslationFilter=(bX=true,bY=false,bZ=false),RotationFilter=(bX=false,bY=false,bZ=false),ScaleFilter=(bX=false,bY=false,bZ=false))
    Parents=((Item=(Type=Bone,Name="spine_05"),Weight=1),(Item=(Type=Bone,Name="head"),Weight=1))
    Weight=1
   )
   > n1 = &RigUnit_SphericalPoseReader(
    ActiveRegionScaleFactors=(PositiveWidth=1.7,NegativeWidth=1.7,PositiveHeight=3,NegativeHeight=3)
    ActiveRegionSize=0.1
    Debug=(bDrawDebug=false,bDraw2D=False,bDrawLocalAxes=False,DebugScale=25,DebugSegments=20,DebugThickness=0.1)
    DriverAxis=(X=1,Y=0,Z=0)
    DriverItem=(Type=Bone,Name="head")
    FalloffRegionScaleFactors=(PositiveWidth=1,NegativeWidth=1,PositiveHeight=1,NegativeHeight=1)
    FalloffSize=0.5
    FlipHeightScaling=False
    FlipWidthScaling=False
    OptionalParentItem=(Type=Bone,Name="spine_05")
    RotationOffset=(X=-90,Y=0,Z=0)
   )
   > &RigUnit_OffsetTransformForItem(
    Item=(Type=Bone,Name="hair_alt_head")
    OffsetTransform=(Rotation=(X=0,Y=0,Z=0.342,W=0.9397),Translation=(X=4,Y=0,Z=0),Scale3D=(X=1,Y=1,Z=1))
    bPropagateToChildren=True
    Weight=@n1.OutputParam
   )
   > n2 = &RigUnit_SphericalPoseReader(
    ActiveRegionScaleFactors=(PositiveWidth=1,NegativeWidth=1,PositiveHeight=2.5,NegativeHeight=2.5)
    ActiveRegionSize=0.15
    Debug=(bDrawDebug=false,bDraw2D=False,bDrawLocalAxes=False,DebugScale=25,DebugSegments=20,DebugThickness=0.1)
    DriverAxis=(X=1,Y=0,Z=0)
    DriverItem=(Type=Bone,Name="head")
    FalloffRegionScaleFactors=(PositiveWidth=1,NegativeWidth=1,PositiveHeight=1,NegativeHeight=1)
    FalloffSize=0.5
    FlipHeightScaling=False
    FlipWidthScaling=False
    OptionalParentItem=(Type=Bone,Name="spine_05")
    RotationOffset=(X=0,Y=0,Z=0)
   )
   > &RigUnit_OffsetTransformForItem(
    Item=(Type=Bone,Name="hair_alt_head")
    OffsetTransform=(Rotation=(X=0,Y=0,Z=0.0872,W=0.9962),Translation=(X=0,Y=0,Z=-4),Scale3D=(X=1,Y=1,Z=1))
    bPropagateToChildren=True
    Weight=@n2.OutputParam
   )
   > &RigUnit_OffsetTransformForItem(
    Item=(Type=Bone,Name="hair_alt_1")
    OffsetTransform=(Rotation=(X=-0.0227,Y=-0.1285,Z=-0.1722,W=0.9764),Translation=(X=-12,Y=-2,Z=0),Scale3D=(X=1,Y=1,Z=1))
    bPropagateToChildren=True
    Weight=@n2.OutputParam
   )
   > &RigUnit_OffsetTransformForItem(
    Item=(Type=Bone,Name="hair_alt_2")
    OffsetTransform=(Rotation=(X=0,Y=0,Z=0,W=1),Translation=(X=-5,Y=0,Z=-1),Scale3D=(X=1,Y=1,Z=1))
    bPropagateToChildren=True
    Weight=@n2.OutputParam
   )
   n3 = &RigUnit_SphericalPoseReader(
    ActiveRegionScaleFactors=(PositiveWidth=1,NegativeWidth=1,PositiveHeight=1,NegativeHeight=1)
    ActiveRegionSize=0.1
    Debug=(bDrawDebug=false,bDraw2D=False,bDrawLocalAxes=False,DebugScale=25,DebugSegments=20,DebugThickness=0.1)
    DriverAxis=(X=0,Y=-1,Z=0)
    DriverItem=(Type=Bone,Name="spine_05")
    FalloffRegionScaleFactors=(PositiveWidth=1,NegativeWidth=1,PositiveHeight=1,NegativeHeight=1)
    FalloffSize=0.5
    FlipHeightScaling=False
    FlipWidthScaling=False
    OptionalParentItem=(Type=Bone,Name="spine_02")
    RotationOffset=(X=0,Y=0,Z=0)
    ExecutePin=@RigVMFunction_Sequence.B
   )
   > &RigUnit_OffsetTransformForItem(
    Item=(Type=Bone,Name="hair_alt_1")
    OffsetTransform=(Rotation=(X=0.1305,Y=0,Z=0,W=0.9914),Translation=(X=0,Y=-2,Z=0),Scale3D=(X=1,Y=1,Z=1))
    bPropagateToChildren=True
    Weight=@n3.OutputParam
   )
   > &RigUnit_OffsetTransformForItem(
    Item=(Type=Bone,Name="hair_alt_2")
    OffsetTransform=(Rotation=(X=-0.4226,Y=0,Z=0,W=0.9063),Translation=(X=0,Y=5,Z=5),Scale3D=(X=1,Y=1,Z=1))
    bPropagateToChildren=True
    Weight=@n3.OutputParam
   )
   > &RigUnit_OffsetTransformForItem(
    Item=(Type=Bone,Name="hair_alt_3")
    OffsetTransform=(Rotation=(X=-0.3213,Y=-0.1919,Z=0.0215,W=0.9271),Translation=(X=0,Y=0,Z=5),Scale3D=(X=1,Y=1,Z=1))
    bPropagateToChildren=True
    Weight=@n3.OutputParam
   )
   > n4 = &RigUnit_SphericalPoseReader(
    ActiveRegionScaleFactors=(PositiveWidth=1,NegativeWidth=1,PositiveHeight=1,NegativeHeight=1)
    ActiveRegionSize=0.1
    Debug=(bDrawDebug=false,bDraw2D=False,bDrawLocalAxes=False,DebugScale=25,DebugSegments=20,DebugThickness=0.1)
    DriverAxis=(X=0,Y=-1,Z=0)
    DriverItem=(Type=Bone,Name="spine_05")
    FalloffRegionScaleFactors=(PositiveWidth=1,NegativeWidth=1,PositiveHeight=1,NegativeHeight=1)
    FalloffSize=0.5
    FlipHeightScaling=False
    FlipWidthScaling=False
    OptionalParentItem=(Type=Bone,Name="spine_02")
    RotationOffset=(X=180,Y=0,Z=0)
   )
   > &RigUnit_OffsetTransformForItem(
    Item=(Type=Bone,Name="hair_alt_1")
    OffsetTransform=(Rotation=(X=-0.1305,Y=0,Z=0,W=0.9914),Translation=(X=0,Y=-2,Z=0),Scale3D=(X=1,Y=1,Z=1))
    bPropagateToChildren=True
    Weight=@n4.OutputParam
   )
   > &RigUnit_OffsetTransformForItem(
    Item=(Type=Bone,Name="hair_alt_2")
    OffsetTransform=(Rotation=(X=0.4226,Y=0,Z=0,W=0.9063),Translation=(X=0,Y=-5,Z=-5),Scale3D=(X=1,Y=1,Z=1))
    bPropagateToChildren=True
    Weight=@n4.OutputParam
   )
   > &RigUnit_OffsetTransformForItem(
    Item=(Type=Bone,Name="hair_alt_3")
    OffsetTransform=(Rotation=(X=0.3498,Y=0.1332,Z=-0.1398,W=0.9167),Translation=(X=0,Y=0,Z=-5),Scale3D=(X=1,Y=1,Z=1))
    bPropagateToChildren=True
    Weight=@n4.OutputParam
   )
   n5 = &RigUnit_SphericalPoseReader(
    ActiveRegionScaleFactors=(PositiveWidth=1,NegativeWidth=1,PositiveHeight=0.5,NegativeHeight=0.5)
    ActiveRegionSize=0.35
    Debug=(bDrawDebug=True,bDraw2D=False,bDrawLocalAxes=False,DebugScale=10,DebugSegments=20,DebugThickness=0.05)
    DriverAxis=(X=1,Y=0,Z=0)
    DriverItem=(Type=Bone,Name="clavicle_l")
    FalloffRegionScaleFactors=(PositiveWidth=1,NegativeWidth=1,PositiveHeight=1,NegativeHeight=1)
    FalloffSize=0.5
    FlipHeightScaling=False
    FlipWidthScaling=False
    OptionalParentItem=(Type=Bone,Name="None")
    RotationOffset=(X=-90,Y=0,Z=0)
    ExecutePin=@RigVMFunction_Sequence.C
   )
   > n6 = &RigUnit_SphericalPoseReader(
    ActiveRegionScaleFactors=(PositiveWidth=1,NegativeWidth=1,PositiveHeight=0.5,NegativeHeight=0.5)
    ActiveRegionSize=0.35
    Debug=(bDrawDebug=True,bDraw2D=False,bDrawLocalAxes=False,DebugScale=10,DebugSegments=20,DebugThickness=0.05)
    DriverAxis=(X=-1,Y=0,Z=0)
    DriverItem=(Type=Bone,Name="clavicle_r")
    FalloffRegionScaleFactors=(PositiveWidth=1,NegativeWidth=1,PositiveHeight=1,NegativeHeight=1)
    FalloffSize=0.5
    FlipHeightScaling=False
    FlipWidthScaling=False
    OptionalParentItem=(Type=Bone,Name="None")
    RotationOffset=(X=90,Y=0,Z=0)
   )
   Maximum = &RigVMFunction_MathDoubleMax(A=@n6.OutputParam, B=@n5.OutputParam)
   > &RigUnit_OffsetTransformForItem(
    Item=(Type=Bone,Name="hair_alt_1")
    OffsetTransform=(Rotation=(X=0,Y=0,Z=0,W=1),Translation=(X=0,Y=-6,Z=0),Scale3D=(X=1,Y=1,Z=1))
    bPropagateToChildren=True
    Weight=@Maximum.Result
   )
   > &RigUnit_OffsetTransformForItem(
    Item=(Type=Bone,Name="hair_alt_2")
    OffsetTransform=(Rotation=(X=0,Y=0,Z=0,W=1),Translation=(X=0,Y=-5,Z=0),Scale3D=(X=1,Y=1,Z=1))
    bPropagateToChildren=True
    Weight=@Maximum.Result
   )
   n7 = &RigUnit_SphericalPoseReader(
    ActiveRegionScaleFactors=(PositiveWidth=1,NegativeWidth=1,PositiveHeight=1,NegativeHeight=1)
    ActiveRegionSize=0.33
    Debug=(bDrawDebug=True,bDraw2D=False,bDrawLocalAxes=False,DebugScale=10,DebugSegments=20,DebugThickness=0.1)
    DriverAxis=(X=-1,Y=0,Z=0)
    DriverItem=(Type=Bone,Name="pelvis")
    FalloffRegionScaleFactors=(PositiveWidth=1,NegativeWidth=1,PositiveHeight=1,NegativeHeight=1)
    FalloffSize=0.5
    FlipHeightScaling=False
    FlipWidthScaling=False
    OptionalParentItem=(Type=Bone,Name="spine_02")
    RotationOffset=(X=-90,Y=0,Z=0)
    ExecutePin=@RigVMFunction_Sequence.D
   )
   > &RigUnit_OffsetTransformForItem(
    Item=(Type=Bone,Name="hair_alt_2")
    OffsetTransform=(Rotation=(X=0,Y=0,Z=0,W=1),Translation=(X=3,Y=-3,Z=0),Scale3D=(X=1,Y=1,Z=1))
    bPropagateToChildren=True
    Weight=@n7.OutputParam
   )
   > &RigUnit_OffsetTransformForItem(
    Item=(Type=Bone,Name="hair_alt_3")
    OffsetTransform=(Rotation=(X=0,Y=0,Z=0.2588,W=0.9659),Translation=(X=-4,Y=-5,Z=0),Scale3D=(X=1,Y=1,Z=1))
    bPropagateToChildren=True
    Weight=@n7.OutputParam
   )
   RigUnit_GetTransform = &RigUnit_GetTransform(Item=(Type=Control,Name="head_ctrl"), bInitial=False)
   RigUnit_SetTransform = &RigUnit_SetTransform(ExecutePin=(), Item=(Type=Bone,Name="head"), Weight=1, bInitial=False, Value=@RigUnit_GetTransform.Transform)
   &RigUnit_SetTransform(
    ExecutePin=()
    Item=(Type=Bone,Name="spine_05")
    Weight=1
    bInitial=False
    Value=&RigUnit_GetTransform(Item=(Type=Control,Name="spine_05_ctrl"), bInitial=False).Transform
   )
  ]}
]
hierarchy [
 root {loc=<0 0 0>, rot=<0 0 0 1>}
 pelvis {parent=root, loc=<0 -0.5301 98.0571>, rot=<0.0127 0.707 -0.0127 -0.707>}
 spine_01 {parent=pelvis, loc=<6.1208 -0.9328 0>, rot=<0 0 0.034 -0.9994>}
 spine_02 {parent=spine_01, loc=<6.5996 -0.2095 0>, rot=<0 0 0.0238 -0.9997>}
 spine_03 {parent=spine_02, loc=<6.4914 -0.1237 0>, rot=<0 0 0.0149 -0.9999>}
 spine_04 {parent=spine_03, loc=<6.9598 0.0794 0>, rot=<0 0 -0.0132 -0.9999>}
 spine_05 {parent=spine_04, loc=<6.9987 0 0>, rot=<0 0 -0.0015 -1>}
 neck_01 {parent=spine_05, loc=<11.3138 0.1529 0>, rot=<0 0 -0.2003 -0.9797>}
 neck_02 {parent=neck_01, loc=<4.5627 0 0>, rot=<0 0 0.0504 -0.9987>}
 head {parent=neck_02, loc=<5.3914 0 0>, rot=<0 0 0.115 -0.9934>}
 MSTR_Hair_r_2 {parent=head, loc=<13.9582 -6.6499 -4.5374>, rot=<0.3835 -0.5267 0.3589 -0.6684>}
 Hair_r_2 {parent=MSTR_Hair_r_2, loc=<0 0 0>, rot=<0 0 0 1>}
 Hair_r_3 {parent=Hair_r_2, loc=<4.8797 0 0.0002>, rot=<-0.1204 -0.5456 0.1031 -0.8229>}
 Hair_r_4 {parent=Hair_r_3, loc=<6.1622 -0.0002 -0.0005>, rot=<0.0288 -0.1843 0.1504 -0.9709>}
 Hair_r_5 {parent=Hair_r_4, loc=<7.4441 0 0>, rot=<-0.0545 -0.17 0.0155 -0.9838>}
 Hair_r_6 {parent=Hair_r_5, loc=<8.2864 0 0>, rot=<0.0155 0.0731 -0.054 -0.9957>}
 Hair_r_7 {parent=Hair_r_6, loc=<9.6565 0 0>, rot=<-0.003 0.0537 -0.023 -0.9983>}
 Hair_r_8 {parent=Hair_r_7, loc=<7.3763 0 0>, rot=<0.0014 0.0579 0.0161 -0.9982>}
 Hair_r_9 {parent=Hair_r_8, loc=<8.4829 0 0>, rot=<-0.0007 0.0122 0.0539 -0.9985>}
 Hair_r_10 {parent=Hair_r_9, loc=<7.4 0 0>, rot=<0 0 0 1>}
 MSTR_Hair_l_2 {parent=head, loc=<13.958 -6.65 4.5374>, rot=<0.6684 -0.3589 0.5267 -0.3835>}
 Hair_l_2 {parent=MSTR_Hair_l_2, loc=<0 0 0>, rot=<0 0 0 1>}
 Hair_l_3 {parent=Hair_l_2, loc=<4.8796 0 0>, rot=<0.1204 -0.5456 -0.1031 -0.8229>}
 Hair_l_4 {parent=Hair_l_3, loc=<6.1616 0 0>, rot=<-0.0288 -0.1843 -0.1504 -0.9709>}
 Hair_l_5 {parent=Hair_l_4, loc=<7.4444 0 0>, rot=<0.0545 -0.17 -0.0155 -0.9838>}
 Hair_l_6 {parent=Hair_l_5, loc=<8.2862 0 0>, rot=<-0.0155 0.0731 0.054 -0.9957>}
 Hair_l_7 {parent=Hair_l_6, loc=<9.6566 0 0>, rot=<0.003 0.0537 0.023 -0.9983>}
 Hair_l_8 {parent=Hair_l_7, loc=<7.3768 0 0>, rot=<-0.0014 0.0579 -0.0161 -0.9982>}
 Hair_l_9 {parent=Hair_l_8, loc=<8.4822 0 0>, rot=<0.0007 0.0122 -0.0539 -0.9985>}
 Hair_l_10 {parent=Hair_l_9, loc=<8.5 0 0>, rot=<0 0 0 1>}
 hair_alt_head {parent=head, loc=<7.4932 -13.9267 0>, rot=<0 0 -0.0451 -0.999>}
 clavicle_l {parent=spine_05, loc=<5.4831 -0.1053 -3.2946>, rot=<0.0465 -0.7289 -0.009 -0.6829>}
 upperarm_l {parent=clavicle_l, loc=<9.8658 -1.2136 -0.4105>, rot=<0.0164 -0.3245 -0.0269 -0.9454>}
 upperarm_twist_01_l {parent=upperarm_l, loc=<7.2135 -0.1134 0>, rot=<0 0 0 1>}
 clavicle_r {parent=spine_05, loc=<5.4831 -0.1053 3.2946>, rot=<0.7289 0.0465 -0.6829 0.009>}
 upperarm_r {parent=clavicle_r, loc=<-9.8658 1.2136 0.4104>, rot=<0.0164 -0.3245 -0.0269 -0.9454>}
 upperarm_twist_01_r {parent=upperarm_r, loc=<-7.2135 0.1134 0>, rot=<0 0 0 -1>}
 MSTR_Hair_l {parent=spine_05, loc=<6.1031 -11.3854 7.2963>, rot=<0.0701 -0.6981 0.7064 -0.094>}
 MSTR_Hair_r {parent=spine_05, loc=<6.103 -11.3855 -7.2963>, rot=<0.094 -0.7064 0.6981 -0.0701>}
 hair_alt_1 {parent=spine_05, loc=<8.2925 -11.2555 0>, rot=<0 0 -0.081 -0.9967>}
 hair_alt_2 {parent=spine_03, loc=<6.3216 -14.1858 0>, rot=<0 0 -0.0956 -0.9954>}
 MSTR_Hair_r_11 {parent=spine_03, loc=<5.3916 -16.1026 -6.1306>, rot=<0.1816 -0.6929 0.6907 -0.0988>}
 MSTR_Hair_r_10 {parent=spine_03, loc=<5.392 -16.1026 6.1306>, rot=<0.0988 -0.6907 0.6929 -0.1816>}
 hair_alt_3 {parent=spine_02, loc=<0.4112 -18.1307 0>, rot=<0 0 -0.148 -0.989>}
 thigh_l {parent=pelvis, loc=<-1.0855 -0.2109 -8.7053>, rot=<0.0009 0.0355 0.0155 -0.9993>}
 thigh_r {parent=pelvis, loc=<-1.0853 -0.2109 8.7053>, rot=<0.0355 -0.0009 0.9993 0.0155>}
 HairAim_Blend {type=Curve, loc=<0 0 0>, rot=<0 0 0 1>}
 head_ctrl {controlType=EULER_TRANSFORM, type=Control, loc=<0 0 0>, rot=<0 0 0 1>}
 spine_05_ctrl {controlType=EULER_TRANSFORM, type=Control, loc=<0 0 0>, rot=<0 0 0 1>}
]
