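# Control Rig for the TruckScale female character's hair that drives the HairAim_Blend curve from the hair's proximity to an aim point.
# It measures the distance between the current position of bone Hair_l_8 and the AimPoint vector variable.
# An inverse remap turns that distance into a blend weight, rising as the bone nears the point, and writes it to the curve.
# Only this left-side bone is sampled, and no bone transforms are modified.
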ControlRigBlueprint /BRCosmetics/Characters/Player/Female/Medium/Bodies/F_MED_TruckScale_Hair/Meshes/Parts/F_MED_TruckScale_Hair_FaceAcc_CtrlRig_HairDistance
graphs [
 RigVMFunctionLibrary {links [], nodes []}
 RigVMModel {nodes [
   RigUnit_BeginExecution = &RigUnit_BeginExecution()
   RigUnit_GetTransform = &RigUnit_GetTransform(Item=(Type=Bone,Name="Hair_l_8"), bInitial=False)
   RigVMFunction_MathVectorDistance = &RigVMFunction_MathVectorDistance(A=@RigUnit_GetTransform.Transform.Translation, B=RigVMVariableNode() {variable=AimPoint}.Value)
   Remap = &RigVMFunction_MathDoubleRemap(SourceMaximum=35, SourceMinimum=20, TargetMaximum=0, TargetMinimum=1, Value=@RigVMFunction_MathVectorDistance.Result)
   /Script/ControlRig.RigUnit_SetCurveValue(Curve=HairAim_Blend, Value=@Remap.Result, ExecutePin=@RigUnit_BeginExecution.ExecutePin) {is=struct RigVMUnitNode}
  ]}
]
hierarchy [
 root {loc=<0 0 0>, rot=<0 0 0 1>}
 pelvis {parent=root, loc=<0 -0.5301 98.0571>, rot=<0.0127 0.707 -0.0127 -0.707>}
 spine_01 {parent=pelvis, loc=<6.1208 -0.9328 0>, rot=<0 0 0.034 -0.9994>}
 spine_02 {parent=spine_01, loc=<6.5996 -0.2095 0>, rot=<0 0 0.0238 -0.9997>}
 spine_03 {parent=spine_02, loc=<6.4914 -0.1237 0>, rot=<0 0 0.0149 -0.9999>}
 spine_04 {parent=spine_03, loc=<6.9598 0.0794 0>, rot=<0 0 -0.0132 -0.9999>}
 spine_05 {parent=spine_04, loc=<6.9987 0 0>, rot=<0 0 -0.0015 -1>}
 neck_01 {parent=spine_05, loc=<11.3138 0.1529 0>, rot=<0 0 -0.2003 -0.9797>}
 neck_02 {parent=neck_01, loc=<4.5627 0 0>, rot=<0 0 0.0504 -0.9987>}
 head {parent=neck_02, loc=<5.3914 0 0>, rot=<0 0 0.115 -0.9934>}
 MSTR_Hair_r_2 {parent=head, loc=<13.9582 -6.6499 -4.5374>, rot=<0.3835 -0.5267 0.3589 -0.6684>}
 Hair_r_2 {parent=MSTR_Hair_r_2, loc=<0 0 0>, rot=<0 0 0 1>}
 Hair_r_3 {parent=Hair_r_2, loc=<4.8797 0 0.0002>, rot=<-0.1204 -0.5456 0.1031 -0.8229>}
 Hair_r_4 {parent=Hair_r_3, loc=<6.1622 -0.0002 -0.0005>, rot=<0.0288 -0.1843 0.1504 -0.9709>}
 Hair_r_5 {parent=Hair_r_4, loc=<7.4441 0 0>, rot=<-0.0545 -0.17 0.0155 -0.9838>}
 Hair_r_6 {parent=Hair_r_5, loc=<8.2864 0 0>, rot=<0.0155 0.0731 -0.054 -0.9957>}
 Hair_r_7 {parent=Hair_r_6, loc=<9.6565 0 0>, rot=<-0.003 0.0537 -0.023 -0.9983>}
 Hair_r_8 {parent=Hair_r_7, loc=<7.3763 0 0>, rot=<0.0014 0.0579 0.0161 -0.9982>}
 Hair_r_9 {parent=Hair_r_8, loc=<8.4829 0 0>, rot=<-0.0007 0.0122 0.0539 -0.9985>}
 Hair_r_10 {parent=Hair_r_9, loc=<7.4 0 0>, rot=<0 0 0 1>}
 MSTR_Hair_l_2 {parent=head, loc=<13.958 -6.65 4.5374>, rot=<0.6684 -0.3589 0.5267 -0.3835>}
 Hair_l_2 {parent=MSTR_Hair_l_2, loc=<0 0 0>, rot=<0 0 0 1>}
 Hair_l_3 {parent=Hair_l_2, loc=<4.8796 0 0>, rot=<0.1204 -0.5456 -0.1031 -0.8229>}
 Hair_l_4 {parent=Hair_l_3, loc=<6.1616 0 0>, rot=<-0.0288 -0.1843 -0.1504 -0.9709>}
 Hair_l_5 {parent=Hair_l_4, loc=<7.4444 0 0>, rot=<0.0545 -0.17 -0.0155 -0.9838>}
 Hair_l_6 {parent=Hair_l_5, loc=<8.2862 0 0>, rot=<-0.0155 0.0731 0.054 -0.9957>}
 Hair_l_7 {parent=Hair_l_6, loc=<9.6566 0 0>, rot=<0.003 0.0537 0.023 -0.9983>}
 Hair_l_8 {parent=Hair_l_7, loc=<7.3768 0 0>, rot=<-0.0014 0.0579 -0.0161 -0.9982>}
 Hair_l_9 {parent=Hair_l_8, loc=<8.4822 0 0>, rot=<0.0007 0.0122 -0.0539 -0.9985>}
 Hair_l_10 {parent=Hair_l_9, loc=<8.5 0 0>, rot=<0 0 0 1>}
 clavicle_l {parent=spine_05, loc=<5.4831 -0.1053 -3.2946>, rot=<0.0465 -0.7289 -0.009 -0.6829>}
 upperarm_l {parent=clavicle_l, loc=<9.8658 -1.2136 -0.4105>, rot=<0.0164 -0.3245 -0.0269 -0.9454>}
 upperarm_twist_01_l {parent=upperarm_l, loc=<7.2135 -0.1134 0>, rot=<0 0 0 1>}
 clavicle_r {parent=spine_05, loc=<5.4831 -0.1053 3.2946>, rot=<0.7289 0.0465 -0.6829 0.009>}
 upperarm_r {parent=clavicle_r, loc=<-9.8658 1.2136 0.4104>, rot=<0.0164 -0.3245 -0.0269 -0.9454>}
 upperarm_twist_01_r {parent=upperarm_r, loc=<-7.2135 0.1134 0>, rot=<0 0 0 -1>}
 MSTR_Hair_l {parent=spine_05, loc=<6.1031 -11.3854 7.2963>, rot=<0.0701 -0.6981 0.7064 -0.094>}
 MSTR_Hair_r {parent=spine_05, loc=<6.103 -11.3855 -7.2963>, rot=<0.094 -0.7064 0.6981 -0.0701>}
 hair_alt_1 {parent=spine_05, loc=<8.2925 -11.2555 0>, rot=<0 0 -0.081 -0.9967>}
 MSTR_Hair_r_11 {parent=spine_03, loc=<5.3916 -16.1026 -6.1306>, rot=<0.1816 -0.6929 0.6907 -0.0988>}
 MSTR_Hair_r_10 {parent=spine_03, loc=<5.392 -16.1026 6.1306>, rot=<0.0988 -0.6907 0.6929 -0.1816>}
 hair_alt_2 {parent=spine_03, loc=<6.3216 -14.1858 0>, rot=<0 0 -0.0956 -0.9954>}
 hair_alt_3 {parent=spine_02, loc=<0.4112 -18.1307 0>, rot=<0 0 -0.148 -0.989>}
 thigh_l {parent=pelvis, loc=<-1.0855 -0.2109 -8.7053>, rot=<0.0009 0.0355 0.0155 -0.9993>}
 thigh_r {parent=pelvis, loc=<-1.0853 -0.2109 8.7053>, rot=<0.0355 -0.0009 0.9993 0.0155>}
 HairAim_Blend {type=Curve, loc=<0 0 0>, rot=<0 0 0 1>}
]
variables [AimPoint {type=FVector, default=(X=0,Y=0,Z=0)}]
