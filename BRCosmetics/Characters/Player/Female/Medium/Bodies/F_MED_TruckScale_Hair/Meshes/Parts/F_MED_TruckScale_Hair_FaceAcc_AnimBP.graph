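# Post-copy Anim Blueprint for the TruckScale female hair/face-accessory part, adding physics-driven hair to the parent mesh pose.
# BackpackState, set from equipped back bling, selects no-backpack, generic-backpack (spine-constrained) or TruckScale-backpack RigidBody chains, with front-end variants.
# Gameplay chains feed a HairDistance Control Rig using a spine_03-relative AimPoint, with HairIsFarFromAim toggling a second physics pass.
# Gravity override adapts to skydiving; disabled RigidBody falls back to an AltSkin Control Rig.
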AnimBlueprint /BRCosmetics/Characters/Player/Female/Medium/Bodies/F_MED_TruckScale_Hair/Meshes/Parts/F_MED_TruckScale_Hair_FaceAcc_AnimBP
target_skeleton=&/F_MED_TruckScale_Hair_FaceAcc_Skeleton
graphs [
 AnimGraph {nodes [
   n1 = Control Rig(Source=Copy Pose From Mesh() {bUseAttachedParent=true}) {
    alphaBoolBlend {blendOption=Linear}
    controlRigAssetReference {blueprintRigClass=&/F_MED_TruckScale_Hair_FaceAcc_CtrlRig_AltSkin.F_MED_TruckScale_Hair_FaceAcc_CtrlRig_AltSkin_C}
   }
   Output Pose(Result=Blend Poses by bool(BlendPose_0=Use cached pose 'Dynamic_Cache'(), BlendPose_1=@n1, bActiveValue=Get bRigidBodyEnabled()) {blendType=Linear}) {layerGroup=None}
   n2 = RigidBody!(
    ComponentPose=Local To Component(LocalPose=Use cached pose 'TruckScale_BackPack'())
    Alpha
    OverrideWorldGravity:Vector=Get Gravity Override()
    ExternalForce
   ) -> (Pose) {
    bOverrideWorldGravity=true
    overridePhysicsAsset=&/F_MED_TruckScale_Hair_FaceAcc_BuiltInBack_Physics
    simulationSpace=BaseBoneSpace
    alphaBoolBlend {blendOption=Linear}
    baseBoneRef {boneName=neck_01}
    componentAppliedLinearAccClamp=<0 0 0>
    simSpaceSettings {
     maxAngularAcceleration=10
     maxAngularVelocity=3
     maxLinearAcceleration=300
     maxLinearVelocity=250
     velocityScaleZ=0.6
     worldAlpha=1
     externalLinearDragV=<0.1 0.1 0.1>
    }
   }
   n3 = Component To Local(ComponentPose=@n2)
   n4 = Use cached pose 'TruckScale_BackPack'()
   n5 = Get HairIsFarFromAim() -> (HairIsFarFromAim:bool)
   n6 = "Transform (Modify) Bone - Bone: MSTR_Hair_r_2"!(
    ComponentPose=Local To Component(LocalPose=Copy Pose From Mesh() {bUseAttachedParent=true})
    bAlphaBoolEnabled:bool=True
    Translation="0,0,0"
    Rotation="0,0,-30"
    Scale="1,1,1"
   ) -> (Pose) {
    alphaInputType=Bool
    rotationMode=BMM_Additive
    rotationSpace=BCS_BoneSpace
    alphaBoolBlend {blendOption=Linear}
    boneToModify {boneName=MSTR_Hair_r_2}
    rotation {roll=-30}
   }
   n7 = "Transform (Modify) Bone - Bone: MSTR_Hair_l_2"!(ComponentPose=@n6, bAlphaBoolEnabled:bool=True, Translation="0,0,0", Rotation="0,0,30", Scale="1,1,1") -> (Pose) {
    alphaInputType=Bool
    rotationMode=BMM_Additive
    rotationSpace=BCS_BoneSpace
    alphaBoolBlend {blendOption=Linear}
    boneToModify {boneName=MSTR_Hair_l_2}
    rotation {roll=30}
   }
   n8 = RigidBody!(
    ComponentPose=@n7
    Alpha
    OverrideWorldGravity=Get Gravity Override()
    ExternalForce
    SimSpaceSettings:Sim Space Settings Structure=Get sim_space_set_FT_backpack() -> (sim_space_set_FT_backpack:Sim Space Settings Structure)
   ) -> (Pose) {
    bOverrideWorldGravity=true
    overridePhysicsAsset=&/F_MED_TruckScale_Hair_FaceAcc_BuiltInBack_FT_Physics
    simulationSpace=BaseBoneSpace
    alphaBoolBlend {blendOption=Linear}
    baseBoneRef {boneName=neck_01}
    componentAppliedLinearAccClamp=<0 0 0>
   }
   n9 = RigidBody!(
    ComponentPose=Local To Component(LocalPose=Use cached pose 'NoBackpackDynamics'())
    Alpha
    OverrideWorldGravity=Get Gravity Override()
    ExternalForce
   ) -> (Pose) {
    bOverrideWorldGravity=true
    overridePhysicsAsset=&/F_MED_TruckScale_Hair_FaceAcc_Physics
    simulationSpace=BaseBoneSpace
    alphaBoolBlend {blendOption=Linear}
    baseBoneRef {boneName=neck_01}
    componentAppliedLinearAccClamp=<0 0 0>
    simSpaceSettings {
     maxAngularAcceleration=10
     maxAngularVelocity=3
     maxLinearAcceleration=300
     maxLinearVelocity=250
     velocityScaleZ=0.6
     worldAlpha=1
     externalLinearDragV=<0.1 0.1 0.1>
    }
   }
   n10 = Component To Local(ComponentPose=@n9)
   n11 = Use cached pose 'NoBackpackDynamics'()
   n12 = Get HairIsFarFromAim()
   n13 = RigidBody!(
    ComponentPose=Local To Component(LocalPose=Copy Pose From Mesh() {bUseAttachedParent=true})
    Alpha
    OverrideWorldGravity=Get Gravity Override()
    ExternalForce
    SimSpaceSettings:Sim Space Settings Structure=Get sim_space_set_FT() -> (sim_space_set_FT:Sim Space Settings Structure)
   ) -> (Pose) {
    bOverrideWorldGravity=true
    overridePhysicsAsset=&/F_MED_TruckScale_Hair_FaceAcc_FT_Physics
    simulationSpace=BaseBoneSpace
    alphaBoolBlend {blendOption=Linear}
    baseBoneRef {boneName=neck_01}
    componentAppliedLinearAccClamp=<0 0 0>
    simSpaceSettings {
     maxAngularAcceleration=25
     maxAngularVelocity=3
     maxLinearAcceleration=400
     maxLinearVelocity=450
     velocityScaleZ=0.8
     worldAlpha=1
     externalLinearDragV=<0.1 0.1 0.1>
    }
   }
   n14 = RigidBody!(
    ComponentPose=Local To Component(LocalPose=Copy Pose From Mesh() {bUseAttachedParent=true})
    Alpha
    OverrideWorldGravity=Get Gravity Override()
    ExternalForce
   ) -> (Pose) {
    bOverrideWorldGravity=true
    overridePhysicsAsset=&/F_MED_TruckScale_Hair_FaceAcc_Backpack_Physics
    simulationSpace=BaseBoneSpace
    alphaBoolBlend {blendOption=Linear}
    baseBoneRef {boneName=neck_01}
    componentAppliedLinearAccClamp=<0 0 0>
    simSpaceSettings {
     maxAngularAcceleration=100
     maxAngularVelocity=2
     maxLinearAcceleration=100
     maxLinearVelocity=100
     velocityScaleZ=0.5
     worldAlpha=1
     externalLinearDragV=<0.1 0.1 0.1>
    }
   }
   n15 = "Constraint - Hair_l_7 (P)"(ComponentPose=@n14, Alpha=1, ConstraintWeights_0=1) {
    alphaBoolBlend {blendOption=Linear}
    boneToModify {boneName=Hair_l_7}
    constraintSetup [{targetBone {boneName=spine_05}}]
    constraintWeights [1]
   }
   n16 = "Constraint - Hair_r_7 (P)"(ComponentPose=@n15, Alpha=1, ConstraintWeights_0=1) {
    alphaBoolBlend {blendOption=Linear}
    boneToModify {boneName=Hair_r_7}
    constraintSetup [{targetBone {boneName=spine_05}}]
    constraintWeights [1]
   }
   n17 = "Constraint - Hair_l_9 (P)"(ComponentPose=@n16, Alpha=1, ConstraintWeights_0=1) {
    alphaBoolBlend {blendOption=Linear}
    boneToModify {boneName=Hair_l_9}
    constraintSetup [{targetBone {boneName=spine_03}}]
    constraintWeights [1]
   }
   n18 = "Constraint - Hair_r_9 (P)"(ComponentPose=@n17, Alpha=1, ConstraintWeights_0=1) {
    alphaBoolBlend {blendOption=Linear}
    boneToModify {boneName=Hair_r_9}
    constraintSetup [{targetBone {boneName=spine_03}}]
    constraintWeights [1]
   }
   n19 = "Transform (Modify) Bone - Bone: Hair_r_7"(ComponentPose=@n18, Alpha=1, Translation="0,0,0", Rotation="-19,0,0", Scale="1,1,1") {
    rotationMode=BMM_Additive
    rotationSpace=BCS_BoneSpace
    alphaBoolBlend {blendOption=Linear}
    boneToModify {boneName=Hair_r_7}
    rotation {pitch=-19}
   }
   n20 = "Transform (Modify) Bone - Bone: Hair_l_7"(ComponentPose=@n19, Alpha=1, Translation="0,0,0", Rotation="-17,0,0", Scale="1,1,1") {
    rotationMode=BMM_Additive
    rotationSpace=BCS_BoneSpace
    alphaBoolBlend {blendOption=Linear}
    boneToModify {boneName=Hair_l_7}
    rotation {pitch=-17}
   }
   n21 = Blend Poses by int!(
    BlendPose_0:Pose Link Structure=Blend Poses by bool(
     BlendPose_0=Component To Local(ComponentPose=@n13)
     BlendPose_1=Blend Poses by bool(BlendPose_0=@n11, BlendPose_1=@n10, BlendTime_0=1, BlendTime_1=0.6, bActiveValue=@n12) {blendType=Linear}
     bActiveValue=Get bIsInFrontEnd()
    ) {blendType=Linear}
    BlendPose_1:Pose Link Structure=Component To Local(ComponentPose=@n20)
    BlendPose_2:Pose Link Structure=Blend Poses by bool(
     BlendPose_0=Component To Local(ComponentPose=@n8)
     BlendPose_1=Blend Poses by bool(BlendPose_0=@n4, BlendPose_1=@n3, BlendTime_0=1, BlendTime_1=0.6, bActiveValue=@n5) {blendType=Linear}
     bActiveValue=Get bIsInFrontEnd()
    ) {blendType=Linear}
    BlendTime_0:float=0.2
    BlendTime_1:float=0.2
    BlendTime_2:float=0.2
    ActiveChildIndex:int=Get BackpackState() -> (BackpackState:int=0)
   ) -> (Pose:Pose Link Structure) {
    is=title AnimGraphNode_BlendListByInt
    blendType=Linear
    blendPose [{linkId=-1, sourceLinkId=-1}, {linkId=-1, sourceLinkId=-1}, {linkId=-1, sourceLinkId=-1}]
    blendTime [0.1,0.1,0.1]
   }
   Save cached pose 'Dynamic_Cache'(Pose=@n21) {cacheName=Dynamic_Cache}
   n22 = RigidBody!(
    ComponentPose=Local To Component(LocalPose=Copy Pose From Mesh() {bUseAttachedParent=true})
    Alpha
    OverrideWorldGravity=Get Gravity Override()
    ExternalForce
    SimSpaceSettings:Sim Space Settings Structure=Get sim_space_set_Gameplay() -> (sim_space_set_Gameplay:Sim Space Settings Structure)
   ) -> (Pose) {
    bOverrideWorldGravity=true
    overridePhysicsAsset=&F_MED_TruckScale_Hair_FaceAcc_Physics
    simulationSpace=BaseBoneSpace
    alphaBoolBlend {blendOption=Linear}
    baseBoneRef {boneName=neck_01}
    componentAppliedLinearAccClamp=<0 0 0>
    simSpaceSettings {
     maxAngularAcceleration=25
     maxAngularVelocity=3
     maxLinearAcceleration=400
     maxLinearVelocity=450
     velocityScaleZ=0.8
     worldAlpha=1
     externalLinearDragV=<0.1 0.1 0.1>
    }
   }
   n23 = Control Rig!(Source=Component To Local(ComponentPose=@n22), bAlphaBoolEnabled:bool=True, AimPoint:Vector=Get AimPoint() -> (AimPoint:Vector)) -> (Pose) {
    alphaInputType=Bool
    alphaBoolBlend {blendOption=Linear}
    controlRigAssetReference {blueprintRigClass=&/F_MED_TruckScale_Hair_FaceAcc_CtrlRig_HairDistance.F_MED_TruckScale_Hair_FaceAcc_CtrlRig_HairDistance_C}
   }
   Save cached pose 'NoBackpackDynamics'(Pose=@n23) {cacheName=NoBackpackDynamics}
   n24 = RigidBody(
    ComponentPose=Local To Component(LocalPose=Copy Pose From Mesh() {bUseAttachedParent=true})
    OverrideWorldGravity=Get Gravity Override()
    SimSpaceSettings=Get sim_space_set_Gameplay()
   ) {
    bOverrideWorldGravity=true
    overridePhysicsAsset=&F_MED_TruckScale_Hair_FaceAcc_BuiltInBack_Physics
    simulationSpace=BaseBoneSpace
    alphaBoolBlend {blendOption=Linear}
    baseBoneRef {boneName=neck_01}
    componentAppliedLinearAccClamp=<0 0 0>
   }
   n25 = Control Rig!(Source=Component To Local(ComponentPose=@n24), Alpha:float, AimPoint=Get AimPoint()) -> (Pose) {
    alphaBoolBlend {blendOption=Linear}
    controlRigAssetReference {blueprintRigClass=&F_MED_TruckScale_Hair_FaceAcc_CtrlRig_HairDistance_C}
   }
   Save cached pose 'TruckScale_BackPack'(Pose=@n25) {cacheName=TruckScale_BackPack}
   Get HairAim_Blend() -> (HairAim_Blend:float)
  ]}
 BlendColors {nodes [
   n1 = BlendColors!() -> (
    then
    Color_1_R:float=~
    Color_1_G:float=~
    Color_1_B:float=~
    Color_1_A:float=~
    Color_2_R:float=~
    Color_2_G:float=~
    Color_2_B:float=~
    Color_2_A:float=~
    Blend:double=~
   ) {is=title K2Node_FunctionEntry}
   > Return Node!(
    execute
    OutColor_R:float=MapRangeClamped(Value=@n1.Blend, InRangeA=0, InRangeB=1, OutRangeA=@n1, OutRangeB=@n1.Color_2_R)
    OutColor_G:float=MapRangeClamped(Value=@n1.Blend, InRangeA=0, InRangeB=1, OutRangeA=@n1.Color_1_G, OutRangeB=@n1.Color_2_G)
    OutColor_B:float=MapRangeClamped(Value=@n1.Blend, InRangeA=0, InRangeB=1, OutRangeA=@n1.Color_1_B, OutRangeB=@n1.Color_2_B)
    OutColor_A:float=MapRangeClamped(Value=@n1.Blend, InRangeA=0, InRangeB=1, OutRangeA=@n1.Color_1_A, OutRangeB=@n1.Color_2_A)
   )
  ]}
 CalculateBoneVelocity {nodes [
   n1 = Get skel_mesh() -> (skel_mesh:Skeletal Mesh Component Object Reference)
   n2 = Get target_bone() -> (target_bone:Name=None)
   n3 = DrawDebugLine!(
    execute
    LineStart:Vector=GetSocketLocation!(self:Scene Component Object Reference=Get skel_mesh(), InSocketName:Name=Get target_bone() -> (target_bone=None)) -> (ReturnValue:Vector)
    LineEnd:Vector=vector + vector(A=GetSocketLocation(self=@n1, InSocketName=@n2), B=Get velocity_new() -> (velocity_new:Vector))
    LineColor:Linear Color Structure=(R=1,G=0.9766,B=0.0601,A=1)
    Duration:float=0.01
    Thickness:float=1.5
    DepthPriority:EDrawDebugSceneDepthPriorityGroup Enum=World
   ) -> (then)
   Return Node!(execute=[@n3.then, @n7.then], Velocity:Vector=Get velocity_new())
   CalculateBoneVelocity!() -> (then, target_bone:Name, mult_velocity:double=~) {is=title K2Node_FunctionEntry}
   > n4 = Set braid_translation_current_(braid_translation_current_:Vector=GetSocketLocation(self=Get skel_mesh(), InSocketName=Get target_bone() -> (target_bone=None))) -> (Output_Get:Vector)
   n5 = GetDeltaSeconds!(self:Anim Instance Object Reference) -> (ReturnValue:float)
   n6 = vector - vector(A=@n4, B=Get braid_translation_prev_() -> (braid_translation_prev_:Vector))
   > Set velocity_new(velocity_new:Vector=vector * vector!(
     A=vector / vector(A:Vector=@n6, B:float=@n5) -> (ReturnValue:Vector)
     B=Get mult_velocity() -> (mult_velocity:double)
     C:Vector="-0.01,-0.01,-0.01"
    ) -> (ReturnValue=~)) -> (Output_Get:Vector)
   > n7 = Set braid_translation_prev_(braid_translation_prev_:Vector=GetSocketLocation(self=Get skel_mesh(), InSocketName=Get target_bone() -> (target_bone=None))) -> (Output_Get:Vector)
  ]}
 EventGraph {nodes [
   Event BlueprintUpdateAnimation()
   > Sequence()
   > n1 = Gravity Override State Function DEPRECATED(
    CharacterPartAnimInstance=Self-Reference()
    GameTimeInSeconds=GetGameTimeInSeconds()
    GravityOverrideParams=Get Gravity Override Params()
    RBAN Params=Get RBAN Params()
   )
   > n2 = Branch(Condition=Get bIsSkyDiving())
   n3 = vector * vector!(A=@n1, B:Vector="2.5,2.5,2.5") -> (ReturnValue=~)
   > n4 = Set Gravity Override(Gravity Override=vector + vector(A=@n3, B="0,0,-2200"))
   n5 = GetCurveValue(CurveName=HairAim_Blend)
   Set HairAim_Blend(execute=[@n4.then, @n9.then], HairAim_Blend:float=@n5) -> (Output_Get:float)
   > Set HairIsFarFromAim(HairIsFarFromAim:bool=OR Boolean(A="Equal (Float)"(A=@n5), B=Get bIsOnGround() -> (bIsOnGround:bool))) -> (Output_Get:bool)
   n6 = Get skel_mesh()
   n7 = GetSocketTransform!(self=@n6, InSocketName=spine_03, TransformSpace=RTS_Component) -> (ReturnValue_Location:Vector, ReturnValue_Rotation:Rotator, ReturnValue_Scale:Vector)
   n8 = vector + vector!(
    A=@n7
    B=vector * vector(A=GetUpVector!(InRot:Rotator=@n7.ReturnValue_Rotation) -> (ReturnValue:Vector), B="40,40,40") -> (ReturnValue=~)
    C:Vector=vector * vector(A=GetForwardVector!(InRot:Rotator=@n7.ReturnValue_Rotation) -> (ReturnValue:Vector), B="40,40,40") -> (ReturnValue=~)
    D:Vector=vector * vector(A=GetRightVector!(InRot:Rotator=@n7.ReturnValue_Rotation) -> (ReturnValue:Vector), B="-20,-20,-20") -> (ReturnValue=~)
   ) -> (ReturnValue)
   > Set AimPoint(AimPoint:Vector=@n8) -> (Output_Get:Vector)
   n9 = Set Gravity Override(execute=@n2.else, Gravity Override=@n3)
   n10 = Event BlueprintInitializeAnimation()
   Set skel_mesh(execute=[@n10.then, @n13.then], skel_mesh:Skeletal Mesh Component Object Reference=GetPartSkeletalMeshComponent()) -> (Output_Get:Skeletal Mesh Component Object Reference)
   > n11 = Branch(Condition=CheckForBackbling())
   > n12 = Branch(Condition=BackpackItemDefinitionHasAnyMetaTag!(Tags:Array of Gameplay Tag Structures=Make Array("[0]"=(TagName="Cosmetics.MetaTags.BackPack_TruckScale")), bRequiresEquipped:bool=false) -> (ReturnValue:bool))
   > Set BackpackState(BackpackState:int=2) -> (Output_Get:int=0)
   Set BackpackState(execute=@n12.else, BackpackState=1) -> (Output_Get=0)
   Set BackpackState(execute=@n11.else, BackpackState=0) -> (Output_Get=0)
   n13 = Event OnCharacterCustomizationLoaded()
   n14 = Branch(Condition=Get bIsInFrontEnd())
   > Set!(
    execute
    sim space set_WorldAlpha:float=1
    sim space set_VelocityScaleZ:float=0.8
    sim space set_DampingAlpha:float=1
    sim space set_MaxLinearVelocity:float=450
    sim space set_MaxAngularVelocity:float=3
    sim space set_MaxLinearAcceleration:float=400
    sim space set_MaxAngularAcceleration:float=50
    sim space set_ExternalLinearDragV:Vector="0.1,0.1,0.1"
    sim space set_ExternalLinearVelocity:Vector="0,0,0"
    sim space set_ExternalAngularVelocity:Vector="0,0,0"
   ) -> (then, Output_Get:Sim Space Settings Structure)
   Set(
    execute=@n14.else
    sim space set_WorldAlpha=1
    sim space set_VelocityScaleZ=0.8
    sim space set_DampingAlpha=1
    sim space set_MaxLinearVelocity=450
    sim space set_MaxAngularVelocity=3
    sim space set_MaxLinearAcceleration=400
    sim space set_MaxAngularAcceleration=25
    sim space set_ExternalLinearDragV="0.1,0.1,0.1"
    sim space set_ExternalLinearVelocity="0,0,0"
    sim space set_ExternalAngularVelocity="0,0,0"
   )
   n15 = Get skel_mesh()
   DrawDebugLine(
    LineStart=GetSocketLocation(self=@n6, InSocketName=Hair_l_9)
    LineEnd=ComposeTransforms!(
     A:"Transform (by ref)"=MakeTransform(Location=@n8, Rotation=@n7.ReturnValue_Rotation)
     B:"Transform (by ref)"=Get World Transform!(self:Scene Component Object Reference=@n15) -> (ReturnValue:Transform) {is=title K2Node_CallFunction}
    ) -> (ReturnValue_Location:Vector, ReturnValue_Rotation:Rotator, ReturnValue_Scale:Vector)
    LineColor=BlendColors!(
     Color_1:Linear Color Structure=(R=0.0059,G=1,B=0.0382,A=1)
     Color_2:Linear Color Structure=(R=1,G=0.0682,B=0,A=1)
     Blend:double=Get HairAim_Blend()
    ) -> (OutColor:Linear Color Structure) {is=title K2Node_CallFunction}
    Duration=0.01
    Thickness=1
    DepthPriority=World
   )
  ]}
]
variables [
 AimPoint {type=Vector}
 BackpackState {type=integer}
 braid_translation_current_ {type=Vector}
 braid_translation_prev_ {type=Vector}
 Gravity Override {type=Vector}
 Gravity Override Params {type=GravityOverrideParamsStruct}
 HairAim_Blend {type=number}
 HairIsFarFromAim {type=boolean}
 RBAN Params {
  type=RBANParamsStruct
  default {
   dBNO {simSpaceSettings {
     maxAngularAcceleration=1000
     maxAngularVelocity=3
     maxLinearAcceleration=300
     maxLinearVelocity=700
     velocityScaleZ=0
     externalLinearDragV=<0 0.5 0>
    }}
   dBNO_Carried {simSpaceSettings {
     maxAngularAcceleration=1000
     maxAngularVelocity=3
     maxLinearAcceleration=300
     maxLinearVelocity=700
     velocityScaleZ=0
     externalLinearDragV=<0 0.5 0>
    }}
   emoteOrMelee {simSpaceSettings {
     maxAngularAcceleration=1000
     maxAngularVelocity=3
     maxLinearAcceleration=300
     maxLinearVelocity=700
     velocityScaleZ=0
     externalLinearDragV=<0 0.5 0>
    }}
   front_End {simSpaceSettings {
     maxAngularAcceleration=1000
     maxAngularVelocity=3
     maxLinearAcceleration=300
     maxLinearVelocity=700
     velocityScaleZ=0
     externalLinearDragV=<0 0.5 0>
    }}
   onGround_Crouch_Moving {simSpaceSettings {
     maxAngularAcceleration=1000
     maxAngularVelocity=3
     maxLinearAcceleration=300
     maxLinearVelocity=700
     velocityScaleZ=0
     externalLinearDragV=<0 0.5 0>
    }}
   onGround_Falling {simSpaceSettings {
     maxAngularAcceleration=1000
     maxAngularVelocity=3
     maxLinearAcceleration=300
     maxLinearVelocity=700
     velocityScaleZ=0
     externalLinearDragV=<0 0.5 0>
    }}
   onGround_Moving {simSpaceSettings {
     maxAngularAcceleration=1000
     maxAngularVelocity=3
     maxLinearAcceleration=300
     maxLinearVelocity=700
     velocityScaleZ=0
     externalLinearDragV=<0 0.5 0>
    }}
   onGround_Standing {simSpaceSettings {
     maxAngularAcceleration=1000
     maxAngularVelocity=3
     maxLinearAcceleration=300
     maxLinearVelocity=700
     velocityScaleZ=0
     externalLinearDragV=<0 0.5 0>
    }}
   rail_Grinding {simSpaceSettings {
     maxAngularAcceleration=1000
     maxAngularVelocity=3
     maxLinearAcceleration=300
     maxLinearVelocity=700
     velocityScaleZ=0
     worldAlpha=0.1
     externalLinearDragV=<0 0.5 0>
    }}
   riding_Moving {simSpaceSettings {
     maxAngularAcceleration=1000
     maxAngularVelocity=3
     maxLinearAcceleration=300
     maxLinearVelocity=700
     velocityScaleZ=0
     externalLinearDragV=<0 0.5 0>
    }}
   riding_Moving_Targeting {simSpaceSettings {
     maxAngularAcceleration=1000
     maxAngularVelocity=3
     maxLinearAcceleration=300
     maxLinearVelocity=700
     velocityScaleZ=0
     externalLinearDragV=<0 0.5 0>
    }}
   riding_Standing {simSpaceSettings {
     maxAngularAcceleration=1000
     maxAngularVelocity=3
     maxLinearAcceleration=300
     maxLinearVelocity=700
     velocityScaleZ=0
     externalLinearDragV=<0 0.5 0>
    }}
   skydive_DivingDown {simSpaceSettings {
     maxAngularAcceleration=1000
     maxAngularVelocity=3
     maxLinearAcceleration=300
     maxLinearVelocity=700
     velocityScaleZ=0
     externalLinearDragV=<0 0.5 0>
    }}
   skydive_FloatingDown {
    componentLinearAccScale=<0 0.5 0>
    componentLinearVelScale=<0 0.5 0>
    simSpaceSettings {
     maxAngularAcceleration=1000
     maxAngularVelocity=3
     maxLinearAcceleration=300
     maxLinearVelocity=700
     velocityScaleZ=0
     externalLinearDragV=<0 0.5 0>
    }
   }
   skydive_GoingUp {simSpaceSettings {
     maxAngularAcceleration=1000
     maxAngularVelocity=3
     maxLinearAcceleration=300
     maxLinearVelocity=700
     velocityScaleZ=0
     externalLinearDragV=<0 0.5 0>
    }}
   skydive_Parachuting {
    componentLinearAccScale=<0 0.1 0.3>
    componentLinearVelScale=<0 0.1 0.3>
    simSpaceSettings {
     maxAngularAcceleration=1000
     maxAngularVelocity=3
     maxLinearAcceleration=300
     maxLinearVelocity=700
     velocityScaleZ=0
     externalLinearDragV=<0 0.5 0>
    }
   }
   skydive_Surfing {simSpaceSettings {
     maxAngularAcceleration=1000
     maxAngularVelocity=3
     maxLinearAcceleration=300
     maxLinearVelocity=700
     velocityScaleZ=0
     externalLinearDragV=<0 0.5 0>
    }}
   swimming {simSpaceSettings {
     maxAngularAcceleration=1000
     maxAngularVelocity=3
     maxLinearAcceleration=300
     maxLinearVelocity=700
     velocityScaleZ=0
     externalLinearDragV=<0 0.5 0>
    }}
   tactical_Sprint {simSpaceSettings {
     maxAngularAcceleration=1000
     maxAngularVelocity=3
     maxLinearAcceleration=300
     maxLinearVelocity=700
     velocityScaleZ=0
     externalLinearDragV=<0 0.5 0>
    }}
   ziplining {simSpaceSettings {
     maxAngularAcceleration=1000
     maxAngularVelocity=3
     maxLinearAcceleration=300
     maxLinearVelocity=700
     velocityScaleZ=0
     externalLinearDragV=<0 0.5 0>
    }}
  }
 }
 sim space set {
  type=SimSpaceSettings
  default {
   dampingAlpha=1
   maxAngularAcceleration=10000
   maxAngularVelocity=10000
   maxLinearAcceleration=10000
   maxLinearVelocity=10000
   velocityScaleZ=1
   worldAlpha=0
   externalLinearDragV=<0 0 0>
  }
 }
 sim_space_set_FT {
  type=SimSpaceSettings
  default {
   dampingAlpha=1
   maxAngularAcceleration=50
   maxAngularVelocity=3
   maxLinearAcceleration=400
   maxLinearVelocity=450
   velocityScaleZ=0.8
   worldAlpha=1
   externalLinearDragV=<0.1 0.1 0.1>
  }
 }
 sim_space_set_FT_backpack {
  type=SimSpaceSettings
  default {
   dampingAlpha=1
   maxAngularAcceleration=70
   maxAngularVelocity=4
   maxLinearAcceleration=400
   maxLinearVelocity=450
   velocityScaleZ=0.8
   worldAlpha=1
   externalLinearDragV=<0.1 0.1 0.1>
  }
 }
 sim_space_set_Gameplay {
  type=SimSpaceSettings
  default {
   dampingAlpha=1
   maxAngularAcceleration=25
   maxAngularVelocity=3
   maxLinearAcceleration=400
   maxLinearVelocity=450
   velocityScaleZ=0.8
   worldAlpha=1
   externalLinearDragV=<0.1 0.1 0.1>
  }
 }
 skel_mesh {type=SkeletalMeshComponent}
 Velocity {type=Vector}
]
